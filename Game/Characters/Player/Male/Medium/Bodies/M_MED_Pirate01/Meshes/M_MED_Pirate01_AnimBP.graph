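# Animation Blueprint for the male medium Pirate01 body that adds secondary motion and corrective posing to its frills and armband.
# From a cached copy pose, grounded RBF Pose Drivers shape frills from clavicle and upper-arm rotation, bypassed while skydiving.
# AnimDynamics then springs armband and frill driver bones, while arm_up_front values weight additive clavicle and right-frill adjustments.
# The Dark Shaman skydiving Pose Driver pair has no input pose connected.
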AnimBlueprint /Game/Characters/Player/Male/Medium/Bodies/M_MED_Pirate01/Meshes/M_MED_Pirate01_AnimBP
target_skeleton=&/M_MED_Pirate01_Skeleton
graphs [
 AnimGraph {nodes [
   n1 = Anim Dynamics - Bone: driver_frill_01_l(Alpha=1, ExternalForce="0,0,0") {
    angularDampingOverride=1
    angularSpringConstant=1500
    bAngularSpring=true
    bGravityOverrideInSimSpace=true
    bOverrideAngularDamping=true
    gravityScale=0.5
    numSolverIterationsPostUpdate=2
    numSolverIterationsPreUpdate=8
    simulationSpace=RootRelative
    alphaBoolBlend {blendOption=Linear}
    boundBone {boneName=driver_frill_01_l}
    componentLinearAccScale=<0.1 0.1 0.65>
    componentLinearVelScale=<0.1 0.1 0.65>
    physicsBodyDefinitions [{
      sphereCollisionRadius=0
      boundBone {boneName=driver_frill_01_l}
      boxExtents=<20 20 10>
      constraintSetup {angularTargetAxis=AxisZ, twistAxis=AxisZ, angularLimitsMax=<0 0 0>, angularLimitsMin=<0 -65 0>, angularTarget=<0 0 0>}
      localJointOffset=<0 0 -10>
     }]
   }
   Anim Dynamics - Bone: driver_frill_02_l(ComponentPose=@n1, Alpha=1, ExternalForce="0,0,0") {
    angularDampingOverride=1
    angularSpringConstant=1500
    bAngularSpring=true
    bGravityOverrideInSimSpace=true
    bOverrideAngularDamping=true
    gravityScale=0.5
    numSolverIterationsPostUpdate=2
    numSolverIterationsPreUpdate=8
    simulationSpace=RootRelative
    alphaBoolBlend {blendOption=Linear}
    boundBone {boneName=driver_frill_02_l}
    componentLinearAccScale=<0.1 0.1 0.65>
    componentLinearVelScale=<0.1 0.1 0.65>
    physicsBodyDefinitions [{
      sphereCollisionRadius=0
      boundBone {boneName=driver_frill_02_l}
      boxExtents=<20 20 10>
      constraintSetup {angularTargetAxis=AxisZ, twistAxis=AxisZ, angularLimitsMax=<0 0 0>, angularLimitsMin=<0 -65 0>, angularTarget=<0 0 0>}
      localJointOffset=<0 0 -10>
     }]
   }
   Anim Dynamics - Bone: driver_frill_09_l(Alpha=1, ExternalForce="0,0,0") {
    angularDampingOverride=1
    angularSpringConstant=1500
    bAngularSpring=true
    bGravityOverrideInSimSpace=true
    bOverrideAngularDamping=true
    gravityScale=0.5
    numSolverIterationsPostUpdate=2
    numSolverIterationsPreUpdate=8
    simulationSpace=RootRelative
    alphaBoolBlend {blendOption=Linear}
    boundBone {boneName=driver_frill_09_l}
    componentLinearAccScale=<0.1 0.1 0.65>
    componentLinearVelScale=<0.1 0.1 0.65>
    physicsBodyDefinitions [{
      sphereCollisionRadius=0
      boundBone {boneName=driver_frill_09_l}
      boxExtents=<20 20 10>
      constraintSetup {angularTargetAxis=AxisZ, twistAxis=AxisZ, angularLimitsMax=<0 0 0>, angularLimitsMin=<0 -65 0>, angularTarget=<0 0 0>}
      localJointOffset=<0 0 -10>
     }]
   }
   n2 = Anim Dynamics - Bone: driver_frill_01_r(Alpha=1, ExternalForce="0,0,0") {
    angularDampingOverride=1
    angularSpringConstant=1500
    bAngularSpring=true
    bGravityOverrideInSimSpace=true
    bOverrideAngularDamping=true
    gravityScale=0.5
    numSolverIterationsPostUpdate=2
    numSolverIterationsPreUpdate=8
    simulationSpace=RootRelative
    alphaBoolBlend {blendOption=Linear}
    boundBone {boneName=driver_frill_01_r}
    componentLinearAccScale=<0.1 0.1 0.65>
    componentLinearVelScale=<0.1 0.1 0.65>
    physicsBodyDefinitions [{
      sphereCollisionRadius=0
      boundBone {boneName=driver_frill_01_r}
      boxExtents=<20 20 10>
      constraintSetup {angularTargetAxis=AxisZ, twistAxis=AxisZ, angularLimitsMax=<0 0 0>, angularLimitsMin=<0 -65 0>, angularTarget=<0 0 0>}
      localJointOffset=<0 0 10>
     }]
   }
   Anim Dynamics - Bone: driver_frill_02_r(ComponentPose=@n2, Alpha=1, ExternalForce="0,0,0") {
    angularDampingOverride=1
    angularSpringConstant=1500
    bAngularSpring=true
    bGravityOverrideInSimSpace=true
    bOverrideAngularDamping=true
    gravityScale=0.5
    numSolverIterationsPostUpdate=2
    numSolverIterationsPreUpdate=8
    simulationSpace=RootRelative
    alphaBoolBlend {blendOption=Linear}
    boundBone {boneName=driver_frill_02_r}
    componentLinearAccScale=<0.1 0.1 0.65>
    componentLinearVelScale=<0.1 0.1 0.65>
    physicsBodyDefinitions [{
      sphereCollisionRadius=0
      boundBone {boneName=driver_frill_02_r}
      boxExtents=<20 20 10>
      constraintSetup {angularTargetAxis=AxisZ, twistAxis=AxisZ, angularLimitsMax=<0 0 0>, angularLimitsMin=<0 -65 0>, angularTarget=<0 0 0>}
      localJointOffset=<0 0 10>
     }]
   }
   n3 = Pose Driver - Source: upperarm_r!(SourcePose:Pose Link Structure) -> (Pose:Pose Link Structure) {
    is=title AnimGraphNode_PoseDriver
    axisLength=1
    poseAsset=/BRCosmetics/Characters/Player/Male/Medium/Bodies/M_MED_DarkShaman/Meshes/M_MED_Dark_Shaman_Skydiving_RBF_PoseAsset
    poseTargets [
     {
      bApplyCustomCurve=false
      bIsHidden=false
      distanceMethod=DefaultMethod
      drivenName=base_pose_skydiving
      functionType=DefaultFunction
      targetScale=0.5851
      boneTransforms [{targetRotation=<rot -64.7854 -3.089 6.0959>, targetTranslation=<-15.5111 0 0>}]
      customCurve {defaultValue=3.4028234663852886e+38, postInfinityExtrap=RCCE_Constant, preInfinityExtrap=RCCE_Constant, keys []}
      targetRotation=<rot 0 0 0>
     }
     {
      bApplyCustomCurve=false
      bIsHidden=false
      distanceMethod=DefaultMethod
      drivenName=arms_out_skydiving
      functionType=DefaultFunction
      targetScale=1
      boneTransforms [{targetRotation=<rot -3.7155 0.3236 -58.8148>, targetTranslation=<-15.5111 0 0>}]
      customCurve {defaultValue=3.4028234663852886e+38, postInfinityExtrap=RCCE_Constant, preInfinityExtrap=RCCE_Constant, keys []}
      targetRotation=<rot 0 0 0>
     }
     {
      bApplyCustomCurve=false
      bIsHidden=false
      distanceMethod=DefaultMethod
      drivenName=arms_up_skydiving
      functionType=DefaultFunction
      targetScale=1
      boneTransforms [{targetRotation=<rot 39.4402 25.0244 -97.0687>, targetTranslation=<-15.5111 0 0>}]
      customCurve {defaultValue=3.4028234663852886e+38, postInfinityExtrap=RCCE_Constant, preInfinityExtrap=RCCE_Constant, keys []}
      targetRotation=<rot 0 0 0>
     }
     {
      bApplyCustomCurve=false
      bIsHidden=false
      distanceMethod=DefaultMethod
      drivenName=arms_back_skydiving
      functionType=DefaultFunction
      targetScale=0.5851
      boneTransforms [{targetRotation=<rot -48.2244 -47.1933 57.9696>, targetTranslation=<-15.5111 0 0>}]
      customCurve {defaultValue=3.4028234663852886e+38, postInfinityExtrap=RCCE_Constant, preInfinityExtrap=RCCE_Constant, keys []}
      targetRotation=<rot 0 0 0>
     }
    ]
    rBFParams {radius=24.3888}
    sourceBones [{boneName=upperarm_r}]
   }
   Pose Driver - Source: upperarm_l!(SourcePose:Pose Link Structure=@n3) -> (Pose:Pose Link Structure) {
    is=title AnimGraphNode_PoseDriver
    poseAsset=&M_MED_Dark_Shaman_Skydiving_RBF_PoseAsset
    poseTargets [
     {
      bApplyCustomCurve=false
      bIsHidden=false
      distanceMethod=DefaultMethod
      drivenName=base_pose_skydiving
      functionType=DefaultFunction
      targetScale=0.6771
      boneTransforms [{targetRotation=<rot -64.7854 -3.089 6.0959>, targetTranslation=<15.5111 0 0>}]
      customCurve {defaultValue=3.4028234663852886e+38, postInfinityExtrap=RCCE_Constant, preInfinityExtrap=RCCE_Constant, keys []}
      targetRotation=<rot 0 0 0>
     }
     {
      bApplyCustomCurve=false
      bIsHidden=false
      distanceMethod=DefaultMethod
      drivenName=arms_out_skydiving
      functionType=DefaultFunction
      targetScale=0.7112
      boneTransforms [{targetRotation=<rot -26.1816 -8.0177 -68.9757>, targetTranslation=<15.5111 0 0>}]
      customCurve {defaultValue=3.4028234663852886e+38, postInfinityExtrap=RCCE_Constant, preInfinityExtrap=RCCE_Constant, keys []}
      targetRotation=<rot 0 0 0>
     }
     {
      bApplyCustomCurve=false
      bIsHidden=false
      distanceMethod=DefaultMethod
      drivenName=arms_up_skydiving
      functionType=DefaultFunction
      targetScale=2
      boneTransforms [{targetRotation=<rot 17.5036 16.2169 -69.177>, targetTranslation=<15.5111 0 0>}]
      customCurve {defaultValue=3.4028234663852886e+38, postInfinityExtrap=RCCE_Constant, preInfinityExtrap=RCCE_Constant, keys []}
      targetRotation=<rot 0 0 0>
     }
     {
      bApplyCustomCurve=false
      bIsHidden=false
      distanceMethod=DefaultMethod
      drivenName=arms_back_skydiving
      functionType=DefaultFunction
      targetScale=0.6771
      boneTransforms [{targetRotation=<rot -41.801 -46.7216 34.921>, targetTranslation=<15.5111 0 0>}]
      customCurve {defaultValue=3.4028234663852886e+38, postInfinityExtrap=RCCE_Constant, preInfinityExtrap=RCCE_Constant, keys []}
      targetRotation=<rot 0 0 0>
     }
    ]
    rBFParams {radius=24.815}
    sourceBones [{boneName=upperarm_l}]
   }
   n4 = Pose Driver - Source: clavicle_r!(SourcePose:Pose Link Structure=Use cached pose 'CopyPose_Cache'()) -> (Pose:Pose Link Structure) {
    is=title AnimGraphNode_PoseDriver
    axisLength=1
    poseAsset=&/M_MED_Pirate_01_Grounded_RBF_PoseAsset
    onlyDriveBones [{boneName=frill_03_r}, {boneName=frill_04_r}, {boneName=frill_05_r}, {boneName=frill_06_r}, {boneName=frill_07_r}]
    poseTargets [
     {
      bApplyCustomCurve=false
      bIsHidden=false
      distanceMethod=DefaultMethod
      drivenName=Pose_0
      functionType=DefaultFunction
      targetScale=0.3689
      boneTransforms [
       {targetRotation=<rot -85.1914 41.4375 143.0618>, targetTranslation=<4.3575 -0.1466 2.3223>}
       {targetRotation=<rot -64.7854 -3.089 6.0959>, targetTranslation=<-15.5111 0 0>}
      ]
      customCurve {defaultValue=3.4028234663852886e+38, postInfinityExtrap=RCCE_Constant, preInfinityExtrap=RCCE_Constant, keys []}
      targetRotation=<rot 0 0 0>
     }
     {
      bApplyCustomCurve=false
      bIsHidden=false
      distanceMethod=DefaultMethod
      drivenName=Pose_1
      functionType=DefaultFunction
      targetScale=0.5252
      boneTransforms [
       {targetRotation=<rot -64.2747 -120.6598 -83.241>, targetTranslation=<4.3575 -0.1466 2.3223>}
       {targetRotation=<rot -23.3709 86.7554 -36.3331>, targetTranslation=<-15.5111 0 0>}
      ]
      customCurve {defaultValue=3.4028234663852886e+38, postInfinityExtrap=RCCE_Constant, preInfinityExtrap=RCCE_Constant, keys []}
      targetRotation=<rot 0 0 0>
     }
     {
      bApplyCustomCurve=false
      bIsHidden=false
      distanceMethod=DefaultMethod
      drivenName=Pose_2
      functionType=DefaultFunction
      targetScale=0.2894
      boneTransforms [
       {targetRotation=<rot -69.838 165.5965 22.6769>, targetTranslation=<4.3575 -0.1466 2.3223>}
       {targetRotation=<rot -75.1088 43.0245 -78.2219>, targetTranslation=<-15.5111 0 0>}
      ]
      customCurve {defaultValue=3.4028234663852886e+38, postInfinityExtrap=RCCE_Constant, preInfinityExtrap=RCCE_Constant, keys []}
      targetRotation=<rot 0 0 0>
     }
     {
      bApplyCustomCurve=false
      bIsHidden=false
      distanceMethod=DefaultMethod
      drivenName=Pose_3
      functionType=DefaultFunction
      targetScale=0.2894
      boneTransforms [
       {targetRotation=<rot -73.4134 -176.9703 4.3592>, targetTranslation=<4.3575 -0.1466 2.3223>}
       {targetRotation=<rot -75.5656 105.7208 -93.8077>, targetTranslation=<-15.5111 0 0>}
      ]
      customCurve {defaultValue=3.4028234663852886e+38, postInfinityExtrap=RCCE_Constant, preInfinityExtrap=RCCE_Constant, keys []}
      targetRotation=<rot 0 0 0>
     }
     {
      bApplyCustomCurve=false
      bIsHidden=false
      distanceMethod=DefaultMethod
      drivenName=Pose_4
      functionType=DefaultFunction
      targetScale=0.3689
      boneTransforms [
       {targetRotation=<rot -87.7993 76.2682 109.0758>, targetTranslation=<4.3575 -0.1466 2.3223>}
       {targetRotation=<rot -69.7589 51.1037 -35.2693>, targetTranslation=<-15.5111 0 0>}
      ]
      customCurve {defaultValue=3.4028234663852886e+38, postInfinityExtrap=RCCE_Constant, preInfinityExtrap=RCCE_Constant, keys []}
      targetRotation=<rot 0 0 0>
     }
     {
      bApplyCustomCurve=false
      bIsHidden=false
      distanceMethod=DefaultMethod
      drivenName=Pose_5
      functionType=DefaultFunction
      targetScale=0.2767
      boneTransforms [
       {targetRotation=<rot -76.2754 172.1506 9.5506>, targetTranslation=<4.3575 -0.1466 2.3223>}
       {targetRotation=<rot -45.0185 104.3596 -56.1379>, targetTranslation=<-15.5111 0 0>}
      ]
      customCurve {defaultValue=3.4028234663852886e+38, postInfinityExtrap=RCCE_Constant, preInfinityExtrap=RCCE_Constant, keys []}
      targetRotation=<rot 0 0 0>
     }
     {
      bApplyCustomCurve=false
      bIsHidden=false
      distanceMethod=DefaultMethod
      drivenName=Pose_6
      functionType=DefaultFunction
      targetScale=0.2767
      boneTransforms [
       {targetRotation=<rot -74.8686 -132.7057 -46.9771>, targetTranslation=<4.3575 -0.1466 2.3223>}
       {targetRotation=<rot -39.6825 113.2505 -70.71>, targetTranslation=<-15.5111 0 0>}
      ]
      customCurve {defaultValue=3.4028234663852886e+38, postInfinityExtrap=RCCE_Constant, preInfinityExtrap=RCCE_Constant, keys []}
      targetRotation=<rot 0 0 0>
     }
     {
      bApplyCustomCurve=false
      bIsHidden=false
      distanceMethod=DefaultMethod
      drivenName=Pose_7
      functionType=DefaultFunction
      targetScale=0.3402
      boneTransforms [
       {targetRotation=<rot -57.8419 -161.5812 -18.7948>, targetTranslation=<4.3575 -0.1466 2.3223>}
       {targetRotation=<rot -66.8684 120.4314 -71.1722>, targetTranslation=<-15.5111 0 0>}
      ]
      customCurve {
       defaultValue=3.4028234663852886e+38
       postInfinityExtrap=RCCE_Constant
       preInfinityExtrap=RCCE_Constant
       keys [
        {
         arriveTangent=0
         arriveTangentWeight=0
         interpMode=RCIM_Linear
         leaveTangent=0
         leaveTangentWeight=0
         tangentMode=RCTM_Auto
         tangentWeightMode=RCTWM_WeightedNone
         time=0
         value=0
        }
        {
         arriveTangent=0
         arriveTangentWeight=0
         interpMode=RCIM_Linear
         leaveTangent=0
         leaveTangentWeight=0
         tangentMode=RCTM_Auto
         tangentWeightMode=RCTWM_WeightedNone
         time=1
         value=1
        }
       ]
      }
      targetRotation=<rot 0 0 0>
     }
     {
      bApplyCustomCurve=false
      bIsHidden=false
      distanceMethod=DefaultMethod
      drivenName=Pose_8
      functionType=DefaultFunction
      targetScale=0.7352
      boneTransforms [
       {targetRotation=<rot -53.6465 -152.7178 -29.1383>, targetTranslation=<4.3575 -0.1466 2.3223>}
       {targetRotation=<rot -17.8084 26.8504 -97.253>, targetTranslation=<-15.5111 0 0>}
      ]
      customCurve {defaultValue=3.4028234663852886e+38, postInfinityExtrap=RCCE_Constant, preInfinityExtrap=RCCE_Constant, keys []}
      targetRotation=<rot 0 0 0>
     }
     {
      bApplyCustomCurve=false
      bIsHidden=false
      distanceMethod=DefaultMethod
      drivenName=Pose_9
      functionType=DefaultFunction
      targetScale=2.25
      boneTransforms [
       {targetRotation=<rot -68.9933 -127.6298 -53.9107>, targetTranslation=<4.3575 -0.1466 2.3223>}
       {targetRotation=<rot 33.2732 89.3676 -89.1674>, targetTranslation=<-15.5111 0 0>}
      ]
      customCurve {
       defaultValue=3.4028234663852886e+38
       postInfinityExtrap=RCCE_Constant
       preInfinityExtrap=RCCE_Constant
       keys [
        {
         arriveTangent=0
         arriveTangentWeight=0
         interpMode=RCIM_Linear
         leaveTangent=0
         leaveTangentWeight=0
         tangentMode=RCTM_Auto
         tangentWeightMode=RCTWM_WeightedNone
         time=0
         value=0
        }
        {
         arriveTangent=0
         arriveTangentWeight=0
         interpMode=RCIM_Linear
         leaveTangent=0
         leaveTangentWeight=0
         tangentMode=RCTM_Auto
         tangentWeightMode=RCTWM_WeightedNone
         time=0.3415
         value=0.8068
        }
        {
         arriveTangent=0
         arriveTangentWeight=0
         interpMode=RCIM_Linear
         leaveTangent=0
         leaveTangentWeight=0
         tangentMode=RCTM_Auto
         tangentWeightMode=RCTWM_WeightedNone
         time=1
         value=1
        }
       ]
      }
      targetRotation=<rot 0 0 0>
     }
     {
      bApplyCustomCurve=false
      bIsHidden=false
      distanceMethod=DefaultMethod
      drivenName=Pose_10
      functionType=DefaultFunction
      targetScale=0.7352
      boneTransforms [
       {targetRotation=<rot -83.622 69.4344 115.2173>, targetTranslation=<4.3575 -0.1466 2.3223>}
       {targetRotation=<rot -11.4935 30.1498 -76.7725>, targetTranslation=<-15.5111 0 0>}
      ]
      customCurve {defaultValue=3.4028234663852886e+38, postInfinityExtrap=RCCE_Constant, preInfinityExtrap=RCCE_Constant, keys []}
      targetRotation=<rot 0 0 0>
     }
     {
      bApplyCustomCurve=false
      bIsHidden=false
      distanceMethod=DefaultMethod
      drivenName=Pose_11
      functionType=DefaultFunction
      targetScale=0.7671
      boneTransforms [
       {targetRotation=<rot -66.8403 -177.8105 2.9442>, targetTranslation=<4.3575 -0.1466 2.3223>}
       {targetRotation=<rot 18.2092 9.4898 -75.6901>, targetTranslation=<-15.5111 0 0>}
      ]
      customCurve {defaultValue=3.4028234663852886e+38, postInfinityExtrap=RCCE_Constant, preInfinityExtrap=RCCE_Constant, keys []}
      targetRotation=<rot 0 0 0>
     }
     {
      bApplyCustomCurve=false
      bIsHidden=false
      distanceMethod=DefaultMethod
      drivenName=Pose_19
      functionType=DefaultFunction
      targetScale=0.5554
      boneTransforms [
       {targetRotation=<rot -47.9536 -145.5199 -31.7961>, targetTranslation=<4.3575 -0.1466 2.3223>}
       {targetRotation=<rot -20.6591 112.0123 -37.7437>, targetTranslation=<-15.5111 0 0>}
      ]
      customCurve {defaultValue=3.4028234663852886e+38, postInfinityExtrap=RCCE_Constant, preInfinityExtrap=RCCE_Constant, keys []}
      targetRotation=<rot 0 0 0>
     }
    ]
    rBFParams {radius=28.4806}
    sourceBones [{boneName=clavicle_r}, {boneName=upperarm_r}]
   }
   n5 = Pose Driver - Source: clavicle_l!(SourcePose:Pose Link Structure=@n4) -> (Pose:Pose Link Structure) {
    is=title AnimGraphNode_PoseDriver
    poseAsset=&M_MED_Pirate_01_Grounded_RBF_PoseAsset
    onlyDriveBones [{boneName=frill_03_l}, {boneName=frill_04_l}, {boneName=frill_05_l}, {boneName=frill_06_l}, {boneName=frill_07_l}]
    poseTargets [
     {
      bApplyCustomCurve=false
      bIsHidden=false
      distanceMethod=DefaultMethod
      drivenName=Pose_0
      functionType=DefaultFunction
      targetScale=0.3972
      boneTransforms [
       {targetRotation=<rot -85.1913 -138.5624 143.0618>, targetTranslation=<4.3575 -0.1466 -2.3223>}
       {targetRotation=<rot -64.7854 -3.089 6.0959>, targetTranslation=<15.5111 0 0>}
      ]
      customCurve {defaultValue=3.4028234663852886e+38, postInfinityExtrap=RCCE_Constant, preInfinityExtrap=RCCE_Constant, keys []}
      targetRotation=<rot 0 0 0>
     }
     {
      bApplyCustomCurve=false
      bIsHidden=false
      distanceMethod=DefaultMethod
      drivenName=Pose_1
      functionType=DefaultFunction
      targetScale=0.871
      boneTransforms [
       {targetRotation=<rot -71.8075 -7.0774 -33.3057>, targetTranslation=<4.3575 -0.1466 -2.3223>}
       {targetRotation=<rot -9.5233 -21.9053 -28.615>, targetTranslation=<15.5111 0 0>}
      ]
      customCurve {defaultValue=3.4028234663852886e+38, postInfinityExtrap=RCCE_Constant, preInfinityExtrap=RCCE_Constant, keys []}
      targetRotation=<rot 0 0 0>
     }
     {
      bApplyCustomCurve=false
      bIsHidden=false
      distanceMethod=DefaultMethod
      drivenName=Pose_2
      functionType=DefaultFunction
      targetScale=0.5255
      boneTransforms [
       {targetRotation=<rot -75.4618 19.5667 -12.8773>, targetTranslation=<4.3575 -0.1466 -2.3223>}
       {targetRotation=<rot -64.9089 79.0369 -47.2079>, targetTranslation=<15.5111 0 0>}
      ]
      customCurve {defaultValue=3.4028234663852886e+38, postInfinityExtrap=RCCE_Constant, preInfinityExtrap=RCCE_Constant, keys []}
      targetRotation=<rot 0 0 0>
     }
     {
      bApplyCustomCurve=false
      bIsHidden=false
      distanceMethod=DefaultMethod
      drivenName=Pose_3
      functionType=DefaultFunction
      targetScale=0.6912
      boneTransforms [
       {targetRotation=<rot -74.7073 55.0349 -49.5566>, targetTranslation=<4.3575 -0.1466 -2.3223>}
       {targetRotation=<rot -36.5188 -9.2495 27.844>, targetTranslation=<15.5111 0 0>}
      ]
      customCurve {defaultValue=3.4028234663852886e+38, postInfinityExtrap=RCCE_Constant, preInfinityExtrap=RCCE_Constant, keys []}
      targetRotation=<rot 0 0 0>
     }
     {
      bApplyCustomCurve=false
      bIsHidden=false
      distanceMethod=DefaultMethod
      drivenName=Pose_4
      functionType=DefaultFunction
      targetScale=0.3972
      boneTransforms [
       {targetRotation=<rot -81.7412 -135.4937 140.2621>, targetTranslation=<4.3575 -0.1466 -2.3223>}
       {targetRotation=<rot -49.8941 -8.1717 1.1726>, targetTranslation=<15.5111 0 0>}
      ]
      customCurve {defaultValue=3.4028234663852886e+38, postInfinityExtrap=RCCE_Constant, preInfinityExtrap=RCCE_Constant, keys []}
      targetRotation=<rot 0 0 0>
     }
     {
      bApplyCustomCurve=false
      bIsHidden=false
      distanceMethod=DefaultMethod
      drivenName=Pose_5
      functionType=DefaultFunction
      targetScale=0.8882
      boneTransforms [
       {targetRotation=<rot -80.8623 39.04 -38.1229>, targetTranslation=<4.3575 -0.1466 -2.3223>}
       {targetRotation=<rot -31.8256 65.8251 -86.0604>, targetTranslation=<15.5111 0 0>}
      ]
      customCurve {defaultValue=3.4028234663852886e+38, postInfinityExtrap=RCCE_Constant, preInfinityExtrap=RCCE_Constant, keys []}
      targetRotation=<rot 0 0 0>
     }
     {
      bApplyCustomCurve=false
      bIsHidden=false
      distanceMethod=DefaultMethod
      drivenName=Pose_6
      functionType=DefaultFunction
      targetScale=0.9332
      boneTransforms [
       {targetRotation=<rot -47.5697 14.692 -0.434>, targetTranslation=<4.3575 -0.1466 -2.3223>}
       {targetRotation=<rot -8.3098 26.4205 -78.9654>, targetTranslation=<15.5111 0 0>}
      ]
      customCurve {defaultValue=3.4028234663852886e+38, postInfinityExtrap=RCCE_Constant, preInfinityExtrap=RCCE_Constant, keys []}
      targetRotation=<rot 0 0 0>
     }
     {
      bApplyCustomCurve=false
      bIsHidden=false
      distanceMethod=DefaultMethod
      drivenName=Pose_7
      functionType=DefaultFunction
      targetScale=0.5255
      boneTransforms [
       {targetRotation=<rot -86.0332 143.397 -137.7672>, targetTranslation=<4.3575 -0.1466 -2.3223>}
       {targetRotation=<rot -67.1606 51.1714 -32.7303>, targetTranslation=<15.5111 0 0>}
      ]
      customCurve {defaultValue=3.4028234663852886e+38, postInfinityExtrap=RCCE_Constant, preInfinityExtrap=RCCE_Constant, keys []}
      targetRotation=<rot 0 0 0>
     }
     {
      bApplyCustomCurve=false
      bIsHidden=false
      distanceMethod=DefaultMethod
      drivenName=Pose_8
      functionType=DefaultFunction
      targetScale=0.4637
      boneTransforms [
       {targetRotation=<rot -84.0122 51.6419 -47.6755>, targetTranslation=<4.3575 -0.1466 -2.3223>}
       {targetRotation=<rot -74.2196 -36.1046 27.6902>, targetTranslation=<15.5111 0 0>}
      ]
      customCurve {defaultValue=3.4028234663852886e+38, postInfinityExtrap=RCCE_Constant, preInfinityExtrap=RCCE_Constant, keys []}
      targetRotation=<rot 0 0 0>
     }
     {
      bApplyCustomCurve=false
      bIsHidden=false
      distanceMethod=DefaultMethod
      drivenName=Pose_9
      functionType=DefaultFunction
      targetScale=2
      boneTransforms [
       {targetRotation=<rot -48.3561 10.1188 -2.4428>, targetTranslation=<4.3575 -0.1466 -2.3223>}
       {targetRotation=<rot 14.5763 54.8781 -98.5188>, targetTranslation=<15.5111 0 0>}
      ]
      customCurve {defaultValue=3.4028234663852886e+38, postInfinityExtrap=RCCE_Constant, preInfinityExtrap=RCCE_Constant, keys []}
      targetRotation=<rot 0 0 0>
     }
     {
      bApplyCustomCurve=false
      bIsHidden=false
      distanceMethod=DefaultMethod
      drivenName=Pose_10
      functionType=DefaultFunction
      targetScale=0.9994
      boneTransforms [
       {targetRotation=<rot -79.2873 99.6394 -110.9318>, targetTranslation=<4.3575 -0.1466 -2.3223>}
       {targetRotation=<rot -14.9047 102.8909 -17.6919>, targetTranslation=<15.5111 0 0>}
      ]
      customCurve {defaultValue=3.4028234663852886e+38, postInfinityExtrap=RCCE_Constant, preInfinityExtrap=RCCE_Constant, keys []}
      targetRotation=<rot 0 0 0>
     }
     {
      bApplyCustomCurve=false
      bIsHidden=false
      distanceMethod=DefaultMethod
      drivenName=Pose_11
      functionType=DefaultFunction
      targetScale=0.9076
      boneTransforms [
       {targetRotation=<rot -61.5845 27.1483 -22.337>, targetTranslation=<4.3575 -0.1466 -2.3223>}
       {targetRotation=<rot 18.262 -7.1845 -72.1573>, targetTranslation=<15.5111 0 0>}
      ]
      customCurve {defaultValue=3.4028234663852886e+38, postInfinityExtrap=RCCE_Constant, preInfinityExtrap=RCCE_Constant, keys []}
      targetRotation=<rot 0 0 0>
     }
     {
      bApplyCustomCurve=false
      bIsHidden=false
      distanceMethod=DefaultMethod
      drivenName=Pose_19
      functionType=DefaultFunction
      targetScale=1
      boneTransforms [
       {targetRotation=<rot -58.4846 -9.5373 2.6759>, targetTranslation=<4.3575 -0.1466 -2.3223>}
       {targetRotation=<rot -44.7022 128.4906 -48.0679>, targetTranslation=<15.5111 0 0>}
      ]
      customCurve {defaultValue=3.4028234663852886e+38, postInfinityExtrap=RCCE_Constant, preInfinityExtrap=RCCE_Constant, keys []}
      targetRotation=<rot 0 0 0>
     }
    ]
    rBFParams {radius=25}
    sourceBones [{boneName=clavicle_l}, {boneName=upperarm_l}]
   }
   n6 = Use cached pose 'CopyPose_Cache'()
   n7 = Get bIsSkyDiving()
   n8 = Anim Dynamics - Bone: dyn_armband_knot(
    ComponentPose=Local To Component(LocalPose=Blend Poses by bool(BlendPose_0=@n6, BlendPose_1=@n5, bActiveValue=@n7) {blendType=Linear})
    Alpha=0.75
    ExternalForce="0,0,0"
   ) {
    alpha=0.9
    angularDampingOverride=0.98
    angularSpringConstant=1500
    bAngularSpring=true
    bGravityOverrideInSimSpace=true
    bOverrideAngularDamping=true
    numSolverIterationsPostUpdate=4
    numSolverIterationsPreUpdate=16
    simulationSpace=Actor
    alphaBoolBlend {blendOption=Linear}
    boundBone {boneName=dyn_armband_knot}
    physicsBodyDefinitions [{
      sphereCollisionRadius=0
      boundBone {boneName=dyn_armband_knot}
      boxExtents=<15 5 5>
      constraintSetup {angularTargetAxis=AxisZ, twistAxis=AxisY, angularLimitsMax=<15 0 15>, angularLimitsMin=<-15 0 -15>, angularTarget=<0 0 1>}
      localJointOffset=<0 0 0>
     }]
   }
   n9 = Anim Dynamics - Bone: dyn_armband_low(ComponentPose=@n8, Alpha=0.5, ExternalForce="0,0,0") {
    alpha=0.9
    angularDampingOverride=0.98
    angularSpringConstant=1500
    bAngularSpring=true
    bGravityOverrideInSimSpace=true
    bOverrideAngularDamping=true
    numSolverIterationsPostUpdate=4
    numSolverIterationsPreUpdate=16
    simulationSpace=Actor
    alphaBoolBlend {blendOption=Linear}
    boundBone {boneName=dyn_armband_low}
    physicsBodyDefinitions [{
      sphereCollisionRadius=0
      boundBone {boneName=dyn_armband_low}
      boxExtents=<15 5 5>
      constraintSetup {angularTargetAxis=AxisZ, twistAxis=AxisY, angularLimitsMax=<15 15 15>, angularLimitsMin=<-15 -15 -15>, angularTarget=<0 0 1>}
      localJointOffset=<0 0 0>
     }]
   }
   n10 = Anim Dynamics - Bone: dyn_armband_hi(ComponentPose=@n9, Alpha=0.5, ExternalForce="0,0,0") {
    alpha=0.9
    angularDampingOverride=0.98
    angularSpringConstant=1500
    bAngularSpring=true
    bGravityOverrideInSimSpace=true
    bOverrideAngularDamping=true
    numSolverIterationsPostUpdate=4
    numSolverIterationsPreUpdate=16
    simulationSpace=Actor
    alphaBoolBlend {blendOption=Linear}
    boundBone {boneName=dyn_armband_hi}
    physicsBodyDefinitions [{
      sphereCollisionRadius=0
      boundBone {boneName=dyn_armband_hi}
      boxExtents=<15 5 5>
      constraintSetup {angularTargetAxis=AxisZ, twistAxis=AxisY, angularLimitsMax=<15 15 15>, angularLimitsMin=<-15 -15 -15>, angularTarget=<0 0 1>}
      localJointOffset=<0 0 0>
     }]
   }
   n11 = Get arm_up_front_r() -> (arm_up_front_r:double)
   n12 = "Transform (Modify) Bone - Bone: clavicle_offset_r"(ComponentPose=@n10, Alpha=@n11, Translation="0,0,0", Rotation="0.000,80,0.000", Scale="1,1,1") {
    rotationMode=BMM_Additive
    rotationSpace=BCS_BoneSpace
    alphaBoolBlend {blendOption=Linear}
    boneToModify {boneName=clavicle_offset_r}
    rotation {yaw=80}
   }
   n13 = "Transform (Modify) Bone - Bone: clavicle_offset_l"(ComponentPose=@n12, Alpha=Get arm_up_front_l() -> (arm_up_front_l:double), Translation="0,0,0", Rotation="0.000,40,0.000", Scale="1,1,1") {
    rotationMode=BMM_Additive
    rotationSpace=BCS_BoneSpace
    alphaBoolBlend {blendOption=Linear}
    boneToModify {boneName=clavicle_offset_l}
    rotation {yaw=40}
   }
   n14 = "Transform (Modify) Bone - Bone: frill_02_r"(ComponentPose=@n13, Alpha=@n11, Translation="0,0,0", Rotation="0.000,30,0.000", Scale="0.55,0.8,0.8") {
    rotationMode=BMM_Additive
    rotationSpace=BCS_BoneSpace
    scaleMode=BMM_Additive
    scaleSpace=BCS_BoneSpace
    alphaBoolBlend {blendOption=Linear}
    boneToModify {boneName=frill_02_r}
    rotation {yaw=30}
    scale=<0.55 0.8 0.8>
   }
   n15 = "Transform (Modify) Bone - Bone: frill_01_r"(ComponentPose=@n14, Alpha=@n11, Translation="0,0,0", Rotation="0,25,0", Scale="0.6,0.8,0.8") {
    rotationMode=BMM_Additive
    rotationSpace=BCS_BoneSpace
    scaleMode=BMM_Additive
    scaleSpace=BCS_BoneSpace
    alphaBoolBlend {blendOption=Linear}
    boneToModify {boneName=frill_01_r}
    rotation {yaw=25}
    scale=<0.6 0.8 0.8>
   }
   n16 = Anim Dynamics - Bone: driver_frill_03_l(ComponentPose=@n15, Alpha=1, ExternalForce="0,0,0") {
    angularDampingOverride=1
    angularSpringConstant=1500
    bAngularSpring=true
    bGravityOverrideInSimSpace=true
    bOverrideAngularDamping=true
    gravityScale=0.5
    numSolverIterationsPostUpdate=2
    numSolverIterationsPreUpdate=8
    simulationSpace=RootRelative
    alphaBoolBlend {blendOption=Linear}
    boundBone {boneName=driver_frill_03_l}
    componentLinearAccScale=<0.1 0.1 0.65>
    componentLinearVelScale=<0.1 0.1 0.65>
    physicsBodyDefinitions [{
      sphereCollisionRadius=0
      boundBone {boneName=driver_frill_03_l}
      boxExtents=<20 20 10>
      constraintSetup {angularTargetAxis=AxisZ, twistAxis=AxisZ, angularLimitsMax=<0 0 0>, angularLimitsMin=<0 -65 0>, angularTarget=<0 0 0>}
      localJointOffset=<0 0 -10>
     }]
   }
   n17 = Anim Dynamics - Bone: driver_frill_04_l(ComponentPose=@n16, Alpha=1, ExternalForce="0,0,0") {
    angularDampingOverride=1
    angularSpringConstant=1500
    bAngularSpring=true
    bGravityOverrideInSimSpace=true
    bOverrideAngularDamping=true
    gravityScale=0.5
    numSolverIterationsPostUpdate=2
    numSolverIterationsPreUpdate=8
    simulationSpace=RootRelative
    alphaBoolBlend {blendOption=Linear}
    boundBone {boneName=driver_frill_04_l}
    componentLinearAccScale=<0.1 0.1 0.65>
    componentLinearVelScale=<0.1 0.1 0.65>
    physicsBodyDefinitions [{
      sphereCollisionRadius=0
      boundBone {boneName=driver_frill_04_l}
      boxExtents=<20 20 10>
      constraintSetup {angularTargetAxis=AxisZ, twistAxis=AxisZ, angularLimitsMax=<0 0 0>, angularLimitsMin=<0 -65 0>, angularTarget=<0 0 0>}
      localJointOffset=<0 0 -10>
     }]
   }
   n18 = Anim Dynamics - Bone: driver_frill_05_l(ComponentPose=@n17, Alpha=1, ExternalForce="0,0,0") {
    angularDampingOverride=1
    angularSpringConstant=1500
    bAngularSpring=true
    bGravityOverrideInSimSpace=true
    bOverrideAngularDamping=true
    gravityScale=0.5
    numSolverIterationsPostUpdate=2
    numSolverIterationsPreUpdate=8
    simulationSpace=RootRelative
    alphaBoolBlend {blendOption=Linear}
    boundBone {boneName=driver_frill_05_l}
    componentLinearAccScale=<0.1 0.1 0.65>
    componentLinearVelScale=<0.1 0.1 0.65>
    physicsBodyDefinitions [{
      sphereCollisionRadius=0
      boundBone {boneName=driver_frill_05_l}
      boxExtents=<20 20 10>
      constraintSetup {angularTargetAxis=AxisZ, twistAxis=AxisZ, angularLimitsMax=<0 0 0>, angularLimitsMin=<0 -65 0>, angularTarget=<0 0 0>}
      localJointOffset=<0 0 -10>
     }]
   }
   n19 = Anim Dynamics - Bone: driver_frill_06_l(ComponentPose=@n18, Alpha=1, ExternalForce="0,0,0") {
    angularDampingOverride=1
    angularSpringConstant=1500
    bAngularSpring=true
    bGravityOverrideInSimSpace=true
    bOverrideAngularDamping=true
    gravityScale=0.5
    numSolverIterationsPostUpdate=2
    numSolverIterationsPreUpdate=8
    simulationSpace=RootRelative
    alphaBoolBlend {blendOption=Linear}
    boundBone {boneName=driver_frill_06_l}
    componentLinearAccScale=<0.1 0.1 0.65>
    componentLinearVelScale=<0.1 0.1 0.65>
    physicsBodyDefinitions [{
      sphereCollisionRadius=0
      boundBone {boneName=driver_frill_06_l}
      boxExtents=<20 20 10>
      constraintSetup {angularTargetAxis=AxisZ, twistAxis=AxisZ, angularLimitsMax=<0 0 0>, angularLimitsMin=<0 -65 0>, angularTarget=<0 0 0>}
      localJointOffset=<0 0 -10>
     }]
   }
   n20 = Anim Dynamics - Bone: driver_frill_07_l(ComponentPose=@n19, Alpha=1, ExternalForce="0,0,0") {
    angularDampingOverride=1
    angularSpringConstant=1500
    bAngularSpring=true
    bGravityOverrideInSimSpace=true
    bOverrideAngularDamping=true
    gravityScale=0.5
    numSolverIterationsPostUpdate=2
    numSolverIterationsPreUpdate=8
    simulationSpace=RootRelative
    alphaBoolBlend {blendOption=Linear}
    boundBone {boneName=driver_frill_07_l}
    componentLinearAccScale=<0.1 0.1 0.65>
    componentLinearVelScale=<0.1 0.1 0.65>
    physicsBodyDefinitions [{
      sphereCollisionRadius=0
      boundBone {boneName=driver_frill_07_l}
      boxExtents=<20 20 10>
      constraintSetup {angularTargetAxis=AxisZ, twistAxis=AxisZ, angularLimitsMax=<0 0 0>, angularLimitsMin=<0 -65 0>, angularTarget=<0 0 0>}
      localJointOffset=<0 0 -10>
     }]
   }
   n21 = Anim Dynamics - Bone: driver_frill_08_l(ComponentPose=@n20, Alpha=1, ExternalForce="0,0,0") {
    angularDampingOverride=1
    angularSpringConstant=1500
    bAngularSpring=true
    bGravityOverrideInSimSpace=true
    bOverrideAngularDamping=true
    gravityScale=0.5
    numSolverIterationsPostUpdate=2
    numSolverIterationsPreUpdate=8
    simulationSpace=RootRelative
    alphaBoolBlend {blendOption=Linear}
    boundBone {boneName=driver_frill_08_l}
    componentLinearAccScale=<0.1 0.1 0.65>
    componentLinearVelScale=<0.1 0.1 0.65>
    physicsBodyDefinitions [{
      sphereCollisionRadius=0
      boundBone {boneName=driver_frill_08_l}
      boxExtents=<20 20 10>
      constraintSetup {angularTargetAxis=AxisZ, twistAxis=AxisZ, angularLimitsMax=<0 0 0>, angularLimitsMin=<0 -65 0>, angularTarget=<0 0 0>}
      localJointOffset=<0 0 -10>
     }]
   }
   n22 = Anim Dynamics - Bone: driver_frill_03_r(ComponentPose=@n21, Alpha=1, ExternalForce="0,0,0") {
    angularDampingOverride=1
    angularSpringConstant=1500
    bAngularSpring=true
    bGravityOverrideInSimSpace=true
    bOverrideAngularDamping=true
    gravityScale=0.5
    numSolverIterationsPostUpdate=2
    numSolverIterationsPreUpdate=8
    simulationSpace=RootRelative
    alphaBoolBlend {blendOption=Linear}
    boundBone {boneName=driver_frill_03_r}
    componentLinearAccScale=<0.1 0.1 0.65>
    componentLinearVelScale=<0.1 0.1 0.65>
    physicsBodyDefinitions [{
      sphereCollisionRadius=0
      boundBone {boneName=driver_frill_03_r}
      boxExtents=<20 20 10>
      constraintSetup {angularTargetAxis=AxisZ, twistAxis=AxisZ, angularLimitsMax=<0 0 0>, angularLimitsMin=<0 -65 0>, angularTarget=<0 0 0>}
      localJointOffset=<0 0 10>
     }]
   }
   n23 = Anim Dynamics - Bone: driver_frill_04_r(ComponentPose=@n22, Alpha=1, ExternalForce="0,0,0") {
    angularDampingOverride=1
    angularSpringConstant=1500
    bAngularSpring=true
    bGravityOverrideInSimSpace=true
    bOverrideAngularDamping=true
    gravityScale=0.5
    numSolverIterationsPostUpdate=2
    numSolverIterationsPreUpdate=8
    simulationSpace=RootRelative
    alphaBoolBlend {blendOption=Linear}
    boundBone {boneName=driver_frill_04_r}
    componentLinearAccScale=<0.1 0.1 0.65>
    componentLinearVelScale=<0.1 0.1 0.65>
    physicsBodyDefinitions [{
      sphereCollisionRadius=0
      boundBone {boneName=driver_frill_04_r}
      boxExtents=<20 20 10>
      constraintSetup {angularTargetAxis=AxisZ, twistAxis=AxisZ, angularLimitsMax=<0 0 0>, angularLimitsMin=<0 -65 0>, angularTarget=<0 0 0>}
      localJointOffset=<0 0 10>
     }]
   }
   n24 = Anim Dynamics - Bone: driver_frill_05_r(ComponentPose=@n23, Alpha=1, ExternalForce="0,0,0") {
    angularDampingOverride=1
    angularSpringConstant=1500
    bAngularSpring=true
    bGravityOverrideInSimSpace=true
    bOverrideAngularDamping=true
    gravityScale=0.5
    numSolverIterationsPostUpdate=2
    numSolverIterationsPreUpdate=8
    simulationSpace=RootRelative
    alphaBoolBlend {blendOption=Linear}
    boundBone {boneName=driver_frill_05_r}
    componentLinearAccScale=<0.1 0.1 0.65>
    componentLinearVelScale=<0.1 0.1 0.65>
    physicsBodyDefinitions [{
      sphereCollisionRadius=0
      boundBone {boneName=driver_frill_05_r}
      boxExtents=<20 20 10>
      constraintSetup {angularTargetAxis=AxisZ, twistAxis=AxisZ, angularLimitsMax=<0 0 0>, angularLimitsMin=<0 -65 0>, angularTarget=<0 0 0>}
      localJointOffset=<0 0 10>
     }]
   }
   n25 = Anim Dynamics - Bone: driver_frill_06_r(ComponentPose=@n24, Alpha=1, ExternalForce="0,0,0") {
    angularDampingOverride=1
    angularSpringConstant=1500
    bAngularSpring=true
    bGravityOverrideInSimSpace=true
    bOverrideAngularDamping=true
    gravityScale=0.5
    numSolverIterationsPostUpdate=2
    numSolverIterationsPreUpdate=8
    simulationSpace=RootRelative
    alphaBoolBlend {blendOption=Linear}
    boundBone {boneName=driver_frill_06_r}
    componentLinearAccScale=<0.1 0.1 0.65>
    componentLinearVelScale=<0.1 0.1 0.65>
    physicsBodyDefinitions [{
      sphereCollisionRadius=0
      boundBone {boneName=driver_frill_06_r}
      boxExtents=<20 20 10>
      constraintSetup {angularTargetAxis=AxisZ, twistAxis=AxisZ, angularLimitsMax=<0 0 0>, angularLimitsMin=<0 -65 0>, angularTarget=<0 0 0>}
      localJointOffset=<0 0 10>
     }]
   }
   n26 = Anim Dynamics - Bone: driver_frill_07_r(ComponentPose=@n25, Alpha=1, ExternalForce="0,0,0") {
    angularDampingOverride=1
    angularSpringConstant=1500
    bAngularSpring=true
    bGravityOverrideInSimSpace=true
    bOverrideAngularDamping=true
    gravityScale=0.5
    numSolverIterationsPostUpdate=2
    numSolverIterationsPreUpdate=8
    simulationSpace=RootRelative
    alphaBoolBlend {blendOption=Linear}
    boundBone {boneName=driver_frill_07_r}
    componentLinearAccScale=<0.1 0.1 0.65>
    componentLinearVelScale=<0.1 0.1 0.65>
    physicsBodyDefinitions [{
      sphereCollisionRadius=0
      boundBone {boneName=driver_frill_07_r}
      boxExtents=<20 20 10>
      constraintSetup {angularTargetAxis=AxisZ, twistAxis=AxisZ, angularLimitsMax=<0 0 0>, angularLimitsMin=<0 -65 0>, angularTarget=<0 0 0>}
      localJointOffset=<0 0 10>
     }]
   }
   n27 = Anim Dynamics - Bone: driver_frill_08_r(ComponentPose=@n26, Alpha=1, ExternalForce="0,0,0") {
    angularDampingOverride=1
    angularSpringConstant=1500
    bAngularSpring=true
    bGravityOverrideInSimSpace=true
    bOverrideAngularDamping=true
    gravityScale=0.5
    numSolverIterationsPostUpdate=2
    numSolverIterationsPreUpdate=8
    simulationSpace=RootRelative
    alphaBoolBlend {blendOption=Linear}
    boundBone {boneName=driver_frill_08_r}
    componentLinearAccScale=<0.1 0.1 0.65>
    componentLinearVelScale=<0.1 0.1 0.65>
    physicsBodyDefinitions [{
      sphereCollisionRadius=0
      boundBone {boneName=driver_frill_08_r}
      boxExtents=<20 20 10>
      constraintSetup {angularTargetAxis=AxisZ, twistAxis=AxisZ, angularLimitsMax=<0 0 0>, angularLimitsMin=<0 -65 0>, angularTarget=<0 0 0>}
      localJointOffset=<0 0 10>
     }]
   }
   n28 = Anim Dynamics - Bone: driver_frill_09_r(ComponentPose=@n27, Alpha=0.45, ExternalForce="0,0,0") {
    angularDampingOverride=1
    angularSpringConstant=1500
    bAngularSpring=true
    bGravityOverrideInSimSpace=true
    bOverrideAngularDamping=true
    gravityScale=0.5
    numSolverIterationsPostUpdate=2
    numSolverIterationsPreUpdate=8
    simulationSpace=RootRelative
    alphaBoolBlend {blendOption=Linear}
    boundBone {boneName=driver_frill_09_r}
    componentLinearAccScale=<0.1 0.1 0.65>
    componentLinearVelScale=<0.1 0.1 0.65>
    physicsBodyDefinitions [{
      sphereCollisionRadius=0
      boundBone {boneName=driver_frill_09_r}
      boxExtents=<20 20 10>
      constraintSetup {angularTargetAxis=AxisZ, twistAxis=AxisZ, angularLimitsMax=<0 0 0>, angularLimitsMin=<0 -65 0>, angularTarget=<0 0 0>}
      localJointOffset=<0 0 10>
     }]
   }
   Output Pose(Result=Component To Local(ComponentPose=@n28)) {layerGroup=None}
   Save cached pose 'CopyPose_Cache'(Pose=Copy Pose From Mesh() {bUseAttachedParent=true}) {cacheName=CopyPose_Cache}
  ]}
 EventGraph {nodes [
   Event BlueprintUpdateAnimation()
   > Sequence()
   n1 = GetPartSkeletalMeshComponent()
   n2 = BreakRotator(InRot=BreakTransform(InTransform=GetDeltaTransformFromRefPose(self=@n1, BoneName=upperarm_r, BaseName=spine_05)).Rotation)
   > Set arm_up_r(arm_up_r:double=MapRangeClamped(Value=@n2.Yaw, InRangeA=80, InRangeB=138)) -> (Output_Get:double)
   > Set arm_up_front_r(arm_up_front_r:double=MapRangeClamped(Value=@n2.Yaw, InRangeA=0, InRangeB=75)) -> (Output_Get:double)
   > Set arm_up_more_r(arm_up_more_r:double=0) -> (Output_Get:double)
   > Set arm_back_r(arm_back_r:double=MapRangeClamped(Value=@n2.Yaw, InRangeA=35, InRangeB=-20)) -> (Output_Get:double)
   n3 = BreakTransform(InTransform=GetDeltaTransformFromRefPose(self=@n1, BoneName=upperarm_r, BaseName=spine_05))
   > Set arm_up_front_l(arm_up_front_l:double=MapRangeClamped(Value=BreakRotator(InRot=@n3.Rotation).Yaw, InRangeA=0, InRangeB=75)) -> (Output_Get:double)
  ]}
]
variables [
 arm_back_r {type=number}
 arm_up_front_l {type=number}
 arm_up_front_r {type=number}
 arm_up_more_r {type=number}
 arm_up_r {type=number}
]
