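# Physics asset for the face accessory mesh of the M_MED_Pirate01 player body, giving its two feathers ragdoll-style physics.
# Capsule bodies cover the head, a small feather base and longer left and right feathers.
# Both feathers are constrained to the base with moderate swing and tight twist limits.
# The base is constrained to the head with free swing and twist, so its angular limit values do not restrict it.
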PhysicsAsset /Game/Characters/Player/Male/Medium/Bodies/M_MED_Pirate01/Meshes/M_MED_Pirate_01_FaceAcc_Physics
bodies {
 dyn_feather_base [{shapeType=Capsule, length=0, radius=0.5, center=<-0.1922 -0.1166 -0.129>}]
 dyn_feather_left [{shapeType=Capsule, length=15.5802, radius=2.688, center=<2.4026 4.0123 -3.4527>, rotation=<rot 0 -33.6067 137.5024>}]
 dyn_feather_right [{shapeType=Capsule, length=12.7309, radius=2.366, center=<-0.0004 4.48 -1.4903>, rotation=<rot 0 0 115.452>}]
 head [{shapeType=Capsule, length=0.1703, radius=1.8576, center=<-0.0645 0.049 -0.0613>}]
}
constraints [
 {bone1Name=dyn_feather_left, bone2Name=dyn_feather_base, swing1LimitDegrees=50, swing2LimitDegrees=50, twistLimitDegrees=10}
 {bone1Name=dyn_feather_right, bone2Name=dyn_feather_base, swing1LimitDegrees=50, swing2LimitDegrees=50, twistLimitDegrees=10}
 {
  bone1Name=dyn_feather_base
  bone2Name=head
  swing1LimitDegrees=45
  swing1Motion=Free
  swing2LimitDegrees=45
  swing2Motion=Free
  twistLimitDegrees=45
  twistMotion=Free
 }
]
